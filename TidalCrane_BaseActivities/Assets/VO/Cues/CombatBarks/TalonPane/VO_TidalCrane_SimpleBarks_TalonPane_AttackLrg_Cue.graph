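# Sound cue for TalonPane's large-attack combat bark voice line in the TidalCrane base activities content.
# A random node picks one of three equally weighted AttackLrg wave variants on each play.
# Playback uses the base activities attenuation, the narrator VO sound class and a TalonPane-specific concurrency set.
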SoundCue /TidalCrane_BaseActivities/Assets/VO/Cues/CombatBarks/TalonPane/VO_TidalCrane_SimpleBarks_TalonPane_AttackLrg_Cue
root=Random_0
nodes [
 {id=Random_0, children [WavePlayer_0,WavePlayer_1,WavePlayer_2], properties {weights [1,1,1]}}
 {
  id=WavePlayer_0
  properties {soundWaveAssetPtr=/TidalCrane_BaseActivities/Assets/VO/Waves/TalonPane/VO_TidalCrane_SimpleBarks_TalonPane_AttackLrg_010}
 }
 {
  id=WavePlayer_1
  properties {soundWaveAssetPtr=/TidalCrane_BaseActivities/Assets/VO/Waves/TalonPane/VO_TidalCrane_SimpleBarks_TalonPane_AttackLrg_011}
 }
 {
  id=WavePlayer_2
  properties {soundWaveAssetPtr=/TidalCrane_BaseActivities/Assets/VO/Waves/TalonPane/VO_TidalCrane_SimpleBarks_TalonPane_AttackLrg_012}
 }
]
properties {
 attenuationSettings=/TidalCrane_BaseActivities/Assets/VO/AS_Fort_BaseActivities
 maxDistance=10100
 soundClassObject=/Game/Sounds/FortSoundClassesAndModes/Fort_VO_Narrator
 volumeMultiplier=1
 concurrencySet [/TidalCrane_BaseActivities/Assets/VO/BaseActivites_VO_TalonPane_Conc]
}
